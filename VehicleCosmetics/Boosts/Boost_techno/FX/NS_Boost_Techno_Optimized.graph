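# Niagara system for the optimized Techno vehicle boost cosmetic, a plasma energy trail made by one emitter with four ribbon renderers.
# VC_ modules spawn one ribbon per filtered boost socket and position it from the skeletal mesh, with width, lifetime and velocity adjusted for frontend preview.
# Per-variant color curves tint the main, R2, R3 and R4 layers through the painted override module.
# Age-scaled sine offsets create a separate wavy lightning position.
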NiagaraSystem /VehicleCosmetics/Boosts/Boost_techno/FX/NS_Boost_Techno_Optimized
dependencies {
 usedDataInterfaces [&NiagaraDataInterfaceColorCurve, &NiagaraDataInterfaceCurve, &NiagaraDataInterfaceSkeletalMesh]
 usedDynamicInputs []
 usedModules [
  &self:SystemSpawnScript.SystemScriptSource.SystemScriptGraph.NiagaraNodeAssignment_0.SetVariables_#1
  &SystemState
 ]
}
emitters {TrailPlasmaEnergy {
  emitterSpawnScript [/SetVariables(
    Emitter.Ribbon 2 Width:NiagaraFloat=&MakeCustomFloatFromBool(Bool=$User.bIsFrontendPreview, True Float=400, False Float=300)
    Emitter.Ribbon 3 Width=22
   ) {isSetParameters=true}]
  emitterUpdateScript [
   &EmitterState(Inactive Response=::NewEnumerator1, Life Cycle Mode=::NewEnumerator1, Loop Duration=1, UseLoopDelay=false)
   &VC_FilteredSocketCount()
   &VC_RateQualityHandler(Epic Spawn Rate=20, High Spawn Rate=20, Low Spawn Rate=20, Medium Spawn Rate=20, Use Custom Rate?=true) {module=VC_RateQualityHandler001}
   &VC_RateQualityHandler()
   &VC_SpawnRibbon_v1(Number of Ribbons=$Emitter.PortCount, Spawn Time=-1, SpawnRate=$Emitter.Spawn Rate)
  ]
  particleSpawnScript [
   &VC_InitializeRibbon_v1(
    Color=$Particles.Color
    Custom Offset=true
    Enable Skeletal Position=true
    Lifetime=&MakeCustomFloatFromBool(Bool=$User.bIsFrontendPreview, True Float=1.2, False Float=1)
    Local Vector Offset=&AddVectorToPosition(
     Position=&MakeCustomVectorFromBool(Bool=$User.bIsFrontendPreview, True Vector=<40 0 0>, False Vector=<40 0 0>)
     Vector=$User.DTCarOffset
    )
    Offset Amount=1
    Ribbon Width=320
   )
   &AddVelocity(Velocity=&MakeCustomVectorFromBool(Bool=$User.bIsFrontendPreview, True Vector=<-1500 0 0>, False Vector=<-1000 0 0>)) {module=AddVelocity001}
  ]
  particleUpdateScript [
   &ParticleState()
   &VC_PaintedVariantsOverride(
    Black=&ColorFromCurve(ColorCurve=DataInterface() {propertyValues {
       AlphaCurve {keys [<at 0.0225 0>, {arriveTangent=1.023, interpMode=RCIM_Cubic, leaveTangent=1.023, time=0.0885, value=0.25}, <at 1 1 RCIM_Cubic>]}
       BlueCurve {keys [<at 0 0 RCIM_Cubic>]}
       GreenCurve {keys [<at 0 0 RCIM_Cubic>]}
       RedCurve {keys [<at 0 0 RCIM_Cubic>]}
      }})
    BurntSienna=&ColorFromCurve(ColorCurve=DataInterface() {propertyValues {
       AlphaCurve {keys [<at 0.0242 0>, {arriveTangent=1.0248, interpMode=RCIM_Cubic, leaveTangent=1.0248, time=0.0868, value=0.9}, <at 1 1 RCIM_Cubic>]}
       BlueCurve {keys [<at 0 0.1001 RCIM_Cubic>]}
       GreenCurve {keys [<at 0 0.1273 RCIM_Cubic>]}
       RedCurve {keys [<at 0 0.25 RCIM_Cubic>]}
      }})
    Cobalt=&ColorFromCurve(ColorCurve=DataInterface() {propertyValues {
       AlphaCurve {keys [<at 0.026 0>, {arriveTangent=1.0267, interpMode=RCIM_Cubic, leaveTangent=1.0267, time=0.0885, value=0.8}, <at 1 1>]}
       BlueCurve {keys [<at 0 0.85 RCIM_Cubic>,<at 0.7771 0.1>]}
       GreenCurve {keys [<at 0 0.0157 RCIM_Cubic>,<at 0.7771 0.0029>]}
       RedCurve {keys [<at 0 0 RCIM_Cubic>,<at 0.7771 0.0002>]}
      }})
    Crimson=&ColorFromCurve(ColorCurve=DataInterface() {propertyValues {
       AlphaCurve {keys [<at 0.0198 0>, {arriveTangent=2.0404, interpMode=RCIM_Cubic, leaveTangent=2.0404, time=0.0806, value=0.8}, <at 1 2 RCIM_Cubic>]}
       BlueCurve {keys [<at 0 0.0186 RCIM_Cubic>,<at 0.7782 0>]}
       GreenCurve {keys [<at 0 0 RCIM_Cubic>,<at 0.7782 0>]}
       RedCurve {keys [<at 0 0.7 RCIM_Cubic>,<at 0.7782 0>]}
      }})
    DefaultColor=&ColorFromCurve(ColorCurve=DataInterface() {propertyValues {
       AlphaCurve {keys [
         <at 0.0225 0>
         {arriveTangent=1.817, interpMode=RCIM_Cubic, leaveTangent=1.817, time=0.0859, value=1}
         <at 0.4627 0.8>
         <at 1 0.2 RCIM_Cubic>
        ]}
       BlueCurve {keys [{arriveTangent=-0.789, interpMode=RCIM_Cubic, time=0, value=0.32}, <at 0.4617 0.0685>]}
       GreenCurve {keys [{arriveTangent=-0.789, interpMode=RCIM_Cubic, time=0, value=0.4}, <at 0.4617 0.0409>]}
       RedCurve {keys [{arriveTangent=-4.4991, interpMode=RCIM_Cubic, time=0, value=0.08}, <at 0.4617 0.0037>]}
      }})
    ForestGreen=&ColorFromCurve(ColorCurve=DataInterface() {propertyValues {
       AlphaCurve {keys [<at 0.0225 0>, {arriveTangent=1.023, interpMode=RCIM_Cubic, leaveTangent=1.023, time=0.0868, value=0.8}, <at 1 1 RCIM_Cubic>]}
       BlueCurve {keys [<at 0 0 RCIM_Cubic>,<at 0.5195 0>]}
       GreenCurve {keys [<at 0 0.5 RCIM_Cubic>,<at 0.5195 0>]}
       RedCurve {keys [<at 0 0.1667 RCIM_Cubic>,<at 0.5195 0>]}
      }})
    Gold:LinearColor=&ColorFromCurve(ColorCurve=DataInterface() {propertyValues {
       AlphaCurve {keys [<at 0 1 RCIM_Cubic>,<at 1 0 RCIM_Cubic>]}
       BlueCurve {keys [<at 0 1 RCIM_Cubic>,<at 1 1 RCIM_Cubic>]}
       GreenCurve {keys [<at 0 1 RCIM_Cubic>,<at 1 1 RCIM_Cubic>]}
       RedCurve {keys [<at 0 1 RCIM_Cubic>,<at 1 1 RCIM_Cubic>]}
      }})
    Grey=&ColorFromCurve(ColorCurve=DataInterface() {propertyValues {
       AlphaCurve {keys [<at 0.026 0>, {arriveTangent=1.0267, interpMode=RCIM_Cubic, leaveTangent=1.0267, time=0.0859, value=1}, <at 1 1 RCIM_Cubic>]}
       BlueCurve {keys [<at 0 0.25 RCIM_Cubic>,<at 0.6478 0>]}
       GreenCurve {keys [<at 0 0.25 RCIM_Cubic>,<at 0.6478 0>]}
       RedCurve {keys [<at 0 0.25 RCIM_Cubic>,<at 0.6478 0>]}
      }})
    Lime=&ColorFromCurve(ColorCurve=DataInterface() {propertyValues {
       AlphaCurve {keys [<at 0.0233 0>,<at 0.0859 1 RCIM_Cubic>]}
       BlueCurve {keys [<at 0 0 RCIM_Cubic>,<at 0.7761 0>]}
       GreenCurve {keys [<at 0 0.4 RCIM_Cubic>,<at 0.7761 0.2>]}
       RedCurve {keys [<at 0 0.2333 RCIM_Cubic>,<at 0.7761 0.1003>]}
      }})
    Orange=&ColorFromCurve(ColorCurve=DataInterface() {propertyValues {
       AlphaCurve {keys [<at 0.0207 0>,<at 0.085 1 RCIM_Cubic>]}
       BlueCurve {keys [<at 0 0 RCIM_Cubic>,<at 0.7782 0>]}
       GreenCurve {keys [<at 0 0.1395 RCIM_Cubic>,<at 0.7782 0.1046>]}
       RedCurve {keys [<at 0 0.4 RCIM_Cubic>,<at 0.7782 0.3>]}
      }})
    Pink=&ColorFromCurve(ColorCurve=DataInterface() {propertyValues {
       AlphaCurve {keys [<at 0.0269 0>,<at 0.0859 1 RCIM_Cubic>]}
       BlueCurve {keys [<at 0 0.3929 RCIM_Cubic>,<at 0.7824 0.1953>]}
       GreenCurve {keys [<at 0 0.2012 RCIM_Cubic>,<at 0.7824 0.1>]}
       RedCurve {keys [<at 0 0.5029 RCIM_Cubic>,<at 0.7824 0.25>]}
      }})
    Purple=&ColorFromCurve(ColorCurve=DataInterface() {propertyValues {
       AlphaCurve {keys [<at 0.0225 0>,<at 0.092 1 RCIM_Cubic>,<at 0.7086 0>]}
       BlueCurve {keys [<at 0 0.85 RCIM_Cubic>,<at 0.796 0.5>]}
       GreenCurve {keys [<at 0 0 RCIM_Cubic>,<at 0.796 0>]}
       RedCurve {keys [<at 0 0.17 RCIM_Cubic>,<at 0.796 0.2917>]}
      }})
    R2 Black:LinearColor=&ColorFromCurve(ColorCurve=DataInterface() {propertyValues {
       AlphaCurve {keys [<at 0.0216 0>, {arriveTangent=2.4021, interpMode=RCIM_Cubic, time=0.0947, value=1}, <at 0.3414 0.8>, <at 0.6392 0.6382>, <at 0.9113 0>]}
       BlueCurve {keys [{arriveTangent=-0.789, interpMode=RCIM_Cubic, time=0, value=0.04}, <at 0.2123 0.02>]}
       GreenCurve {keys [{arriveTangent=-0.789, interpMode=RCIM_Cubic, time=0, value=0.04}, <at 0.2123 0.02>]}
       RedCurve {keys [{arriveTangent=-4.4991, interpMode=RCIM_Cubic, time=0, value=0.04}, <at 0.2123 0.02>]}
      }})
    R2 BurntSienna:LinearColor=&ColorFromCurve(ColorCurve=DataInterface() {propertyValues {
       AlphaCurve {keys [<at 0.0233 0>, {arriveTangent=2.1641, interpMode=RCIM_Cubic, time=0.092, value=1}, <at 0.4801 1>, <at 0.9006 0 RCIM_Cubic>]}
       BlueCurve {keys [<at 0 0.05>, {arriveTangent=-0.0475, interpMode=RCIM_Cubic, leaveTangent=-0.0475, time=0.2902, value=0.0194}, <at 0.9945 0.0027>]}
       GreenCurve {keys [<at 0 0.0608>, {arriveTangent=-0.0587, interpMode=RCIM_Cubic, leaveTangent=-0.0587, time=0.2902, value=0.0292}, <at 0.9945 0.0024>]}
       RedCurve {keys [<at 0 0.1>, {arriveTangent=-0.086, interpMode=RCIM_Cubic, leaveTangent=-0.086, time=0.2902, value=0.0648}, <at 0.9945 0.0144>]}
      }})
    R2 Cobalt:LinearColor=&ColorFromCurve(ColorCurve=DataInterface() {propertyValues {
       AlphaCurve {keys [<at 0.0216 0>, {arriveTangent=2.0804, interpMode=RCIM_Cubic, time=0.0903, value=1}, <at 0.4908 1>, <at 1 -0.5 RCIM_Cubic>]}
       BlueCurve {keys [<at 0 0.4 RCIM_Cubic>,<at 0.3468 0.14>]}
       GreenCurve {keys [<at 0 0.076 RCIM_Cubic>,<at 0.3468 0.0518>]}
       RedCurve {keys [<at 0 0.04 RCIM_Cubic>,<at 0.3468 0.042>]}
      }})
    R2 Crimson:LinearColor=&ColorFromCurve(ColorCurve=DataInterface() {propertyValues {
       AlphaCurve {keys [
         <at 0.0181 0>
         {arriveTangent=2.1782, interpMode=RCIM_Cubic, time=0.0912, value=1}
         {arriveTangent=-1.2338, interpMode=RCIM_Cubic, leaveTangent=-1.2338, time=0.4631, value=1}
         <at 0.9017 0>
        ]}
       BlueCurve {keys [<at 0 0.0106 RCIM_Cubic>,<at 0.2293 0.004>,<at 0.9593 0>]}
       GreenCurve {keys [<at 0 0 RCIM_Cubic>,<at 0.2293 0>,<at 0.9593 0>]}
       RedCurve {keys [<at 0 0.4 RCIM_Cubic>,<at 0.2293 0.15>,<at 0.9593 0.05>]}
      }})
    R2 DefaultColor:LinearColor=&ColorFromCurve(ColorCurve=DataInterface() {propertyValues {
       AlphaCurve {keys [<at 0.0242 0 RCIM_Cubic>,<at 0.0973 0.9889>,<at 0.4531 0.8>,<at 1 -0.1>]}
       BlueCurve {keys [{arriveTangent=-0.789, interpMode=RCIM_Cubic, time=0, value=0.2}, <at 0.2804 0.1>]}
       GreenCurve {keys [{arriveTangent=-0.789, interpMode=RCIM_Cubic, time=0, value=0.25}, <at 0.2804 0.0575>]}
       RedCurve {keys [{arriveTangent=-4.4991, interpMode=RCIM_Cubic, time=0, value=0.05}, <at 0.2804 0>]}
      }})
    R2 ForestGreen:LinearColor=&ColorFromCurve(ColorCurve=DataInterface() {propertyValues {
       AlphaCurve {keys [
         <at 0.0198 0>
         {arriveTangent=2.195, interpMode=RCIM_Cubic, time=0.0806, value=1}
         {arriveTangent=-1.2179, interpMode=RCIM_Cubic, leaveTangent=-1.2179, time=0.4631, value=1}
         <at 0.9017 0>
        ]}
       BlueCurve {keys [<at 0 0.0013 RCIM_Cubic>,<at 0.2293 0.0005>,<at 0.9593 0.0003>]}
       GreenCurve {keys [<at 0 0.3 RCIM_Cubic>,<at 0.2293 0.07>,<at 0.9593 0.0545>]}
       RedCurve {keys [<at 0 0 RCIM_Cubic>,<at 0.2293 0>,<at 0.9593 0>]}
      }})
    R2 Grey:LinearColor=&ColorFromCurve(ColorCurve=DataInterface() {propertyValues {
       AlphaCurve {keys [
         <at 0.0198 0>
         {arriveTangent=2.1595, interpMode=RCIM_Cubic, time=0.085, value=1}
         {arriveTangent=-1.2244, interpMode=RCIM_Cubic, leaveTangent=-1.2244, time=0.4631, value=1}
         <at 0.9017 0>
        ]}
       BlueCurve {keys [<at 0 0.12 RCIM_Cubic>,<at 0.2293 0.05>,<at 0.9593 0.015>]}
       GreenCurve {keys [<at 0 0.12 RCIM_Cubic>,<at 0.2293 0.05>,<at 0.9593 0.015>]}
       RedCurve {keys [<at 0 0.12 RCIM_Cubic>,<at 0.2293 0.05>,<at 0.9593 0.015>]}
      }})
    R2 Lime:LinearColor=&ColorFromCurve(ColorCurve=DataInterface() {propertyValues {
       AlphaCurve {keys [
         <at 0.0242 0>
         {arriveTangent=2.1993, interpMode=RCIM_Cubic, time=0.0868, value=1}
         {arriveTangent=-1.2271, interpMode=RCIM_Cubic, leaveTangent=-1.2271, time=0.4631, value=1}
         <at 0.9017 0>
        ]}
       BlueCurve {keys [<at 0 0 RCIM_Cubic>,<at 0.2293 0>,<at 0.9593 0>]}
       GreenCurve {keys [<at 0 0.25 RCIM_Cubic>,<at 0.2293 0.13>,<at 0.9593 0.06>]}
       RedCurve {keys [<at 0 0.1589 RCIM_Cubic>,<at 0.2293 0.0883>,<at 0.9593 0.0389>]}
      }})
    R2 Orange:LinearColor=&ColorFromCurve(ColorCurve=DataInterface() {propertyValues {
       AlphaCurve {keys [
         <at 0.0225 0>
         {arriveTangent=2.2696, interpMode=RCIM_Cubic, leaveTangent=2.2696, time=0.0876, value=1}
         {arriveTangent=-1.2284, interpMode=RCIM_Cubic, leaveTangent=-1.2284, time=0.4631, value=1}
         <at 0.9017 0>
        ]}
       BlueCurve {keys [<at 0 0 RCIM_Cubic>,<at 0.2036 0.0085>,<at 0.9593 0>]}
       GreenCurve {keys [<at 0 0.1014 RCIM_Cubic>,<at 0.2036 0.0655>,<at 0.9593 0>]}
       RedCurve {keys [<at 0 0.3 RCIM_Cubic>,<at 0.2036 0.17>,<at 0.9593 0.02>]}
      }})
    R2 Pink:LinearColor=&ColorFromCurve(ColorCurve=DataInterface() {propertyValues {
       AlphaCurve {keys [
         <at 0.0207 0>
         {arriveTangent=1.7731, interpMode=RCIM_Cubic, time=0.0815, value=1}
         {arriveTangent=-1.0887, interpMode=RCIM_Cubic, leaveTangent=-1.0887, time=0.4631, value=0.8}
         <at 1 0>
        ]}
       BlueCurve {keys [<at 0 0.255 RCIM_Cubic>,<at 0.2293 0.1275>,<at 1 0.0105>]}
       GreenCurve {keys [<at 0 0.15 RCIM_Cubic>,<at 0.2293 0.075>,<at 1 0.005>]}
       RedCurve {keys [<at 0 0.3 RCIM_Cubic>,<at 0.2293 0.15>,<at 1 0.05>]}
      }})
    R2 Purple:LinearColor=&ColorFromCurve(ColorCurve=DataInterface() {propertyValues {
       AlphaCurve {keys [
         <at 0.0207 0>
         {arriveTangent=1.8084, interpMode=RCIM_Cubic, leaveTangent=1.8084, time=0.0806, value=1}
         {arriveTangent=-1.3052, interpMode=RCIM_Cubic, leaveTangent=-1.3052, time=0.4631, value=0.8}
         <at 1 -0.2>
        ]}
       BlueCurve {keys [<at 0 0.3 RCIM_Cubic>,<at 0.2293 0.1195>,<at 0.9593 0.07>]}
       GreenCurve {keys [<at 0 0.105 RCIM_Cubic>,<at 0.2293 0.0478>,<at 0.9593 0>]}
       RedCurve {keys [<at 0 0.1282 RCIM_Cubic>,<at 0.2293 0.0538>,<at 0.9593 0.0291>]}
      }})
    R2 Saffron:LinearColor=&ColorFromCurve(ColorCurve=DataInterface() {propertyValues {
       AlphaCurve {keys [
         <at 0.0207 0>
         {arriveTangent=2.2295, interpMode=RCIM_Cubic, time=0.0859, value=1}
         {arriveTangent=-1.3127, interpMode=RCIM_Cubic, leaveTangent=-1.3127, time=0.4631, value=1}
         <at 1 -0.2>
        ]}
       BlueCurve {keys [<at 0 0.0137 RCIM_Cubic>,<at 0.2293 0.0021>,<at 0.9593 0>]}
       GreenCurve {keys [<at 0 0.4 RCIM_Cubic>,<at 0.2293 0.13>,<at 0.9593 0.07>]}
       RedCurve {keys [<at 0 0.3964 RCIM_Cubic>,<at 0.2293 0.13>,<at 0.9593 0.07>]}
      }})
    R2 SkyBlue:LinearColor=&ColorFromCurve(ColorCurve=DataInterface() {propertyValues {
       AlphaCurve {keys [
         <at 0.0225 0>
         {arriveTangent=2.0426, interpMode=RCIM_Cubic, leaveTangent=2.0426, time=0.0885, value=0.5}
         {arriveTangent=-0.768, interpMode=RCIM_Cubic, leaveTangent=-0.768, time=0.4631, value=0.9}
         <at 1 -0.2>
        ]}
       BlueCurve {keys [<at 0 0.4 RCIM_Cubic>,<at 0.2293 0.15>,<at 0.9593 0.08>]}
       GreenCurve {keys [<at 0 0.2 RCIM_Cubic>,<at 0.2293 0.09>,<at 0.9593 0.0666>]}
       RedCurve {keys [<at 0 0 RCIM_Cubic>,<at 0.2293 0.06>,<at 0.9593 0.04>]}
      }})
    R2 TitaniumWhite:LinearColor=&ColorFromCurve(ColorCurve=DataInterface() {propertyValues {
       AlphaCurve {keys [
         <at 0.0207 0>
         {arriveTangent=1.5824, interpMode=RCIM_Cubic, leaveTangent=1.5824, time=0.0841, value=1}
         {arriveTangent=-1.0918, interpMode=RCIM_Cubic, leaveTangent=-1.0918, time=0.4631, value=0.7}
         <at 1 0>
        ]}
       BlueCurve {keys [<at 0 0.25 RCIM_Cubic>,<at 0.2293 0.25>,<at 0.8252 0.1>]}
       GreenCurve {keys [<at 0 0.25 RCIM_Cubic>,<at 0.2293 0.25>,<at 0.8252 0.1>]}
       RedCurve {keys [<at 0 0.25 RCIM_Cubic>,<at 0.2293 0.25>,<at 0.8252 0.1>]}
      }})
    R3 Black:LinearColor=&ColorFromCurve(ColorCurve=DataInterface() {propertyValues {
       AlphaCurve {keys [<at 0 15 RCIM_Cubic>,<at 1 0 RCIM_Cubic>]}
       BlueCurve {keys [<at 0 0 RCIM_Cubic>]}
       GreenCurve {keys [<at 0 0 RCIM_Cubic>]}
       RedCurve {keys [<at 0 0 RCIM_Cubic>]}
      }})
    R3 BurntSienna:LinearColor=&ColorFromCurve(ColorCurve=DataInterface() {propertyValues {
       AlphaCurve {keys [<at 0 7 RCIM_Cubic>,<at 1 0 RCIM_Cubic>]}
       BlueCurve {keys [<at 0 0.0248 RCIM_Cubic>,<at 0.2568 0 RCIM_Cubic>]}
       GreenCurve {keys [<at 0 0.0626 RCIM_Cubic>,<at 0.2568 0 RCIM_Cubic>]}
       RedCurve {keys [<at 0 0.2 RCIM_Cubic>,<at 0.2568 0 RCIM_Cubic>]}
      }})
    R3 Cobalt:LinearColor=&ColorFromCurve(ColorCurve=DataInterface() {propertyValues {
       AlphaCurve {keys [<at 0 5 RCIM_Cubic>,<at 1 0 RCIM_Cubic>]}
       BlueCurve {keys [<at 0 0.3 RCIM_Cubic>]}
       GreenCurve {keys [<at 0 0.0525 RCIM_Cubic>]}
       RedCurve {keys [<at 0 0.03 RCIM_Cubic>]}
      }})
    R3 Crimson:LinearColor=&ColorFromCurve(ColorCurve=DataInterface() {propertyValues {
       AlphaCurve {keys [<at 0 8 RCIM_Cubic>,<at 1 0 RCIM_Cubic>]}
       BlueCurve {keys [<at 0 0.003 RCIM_Cubic>]}
       GreenCurve {keys [<at 0 0 RCIM_Cubic>]}
       RedCurve {keys [<at 0 0.22 RCIM_Cubic>]}
      }})
    R3 DefaultColor:LinearColor=&ColorFromCurve(ColorCurve=DataInterface() {propertyValues {
       AlphaCurve {keys [<at 0 1 RCIM_Cubic>,<at 0.842 0 RCIM_Cubic>]}
       BlueCurve {keys [<at 0 0.9911 RCIM_Cubic>,<at 0.8808 1>]}
       GreenCurve {keys [<at 0 0.8879 RCIM_Cubic>,<at 0.8808 0.9271>]}
       RedCurve {keys [<at 0 0.0009 RCIM_Cubic>,<at 0.8808 0.3>]}
      }})
    R3 ForestGreen:LinearColor=&ColorFromCurve(ColorCurve=DataInterface() {propertyValues {
       AlphaCurve {keys [<at 0 8 RCIM_Cubic>,<at 1 0 RCIM_Cubic>]}
       BlueCurve {keys [<at 0 0.0031 RCIM_Cubic>]}
       GreenCurve {keys [<at 0 0.19 RCIM_Cubic>]}
       RedCurve {keys [<at 0 0 RCIM_Cubic>]}
      }})
    R3 Grey:LinearColor=&ColorFromCurve(ColorCurve=DataInterface() {propertyValues {
       AlphaCurve {keys [<at 0 9 RCIM_Cubic>,<at 1 0 RCIM_Cubic>]}
       BlueCurve {keys [<at 0 1 RCIM_Cubic>,<at 0.2165 0.02 RCIM_Cubic>]}
       GreenCurve {keys [<at 0 1 RCIM_Cubic>,<at 0.2165 0.02 RCIM_Cubic>]}
       RedCurve {keys [<at 0 1 RCIM_Cubic>,<at 0.2165 0.02 RCIM_Cubic>]}
      }})
    R3 Lime:LinearColor=&ColorFromCurve(ColorCurve=DataInterface() {propertyValues {
       AlphaCurve {keys [<at 0 8 RCIM_Cubic>,<at 1 0 RCIM_Cubic>]}
       BlueCurve {keys [<at 0 0 RCIM_Cubic>]}
       GreenCurve {keys [<at 0 0.45 RCIM_Cubic>]}
       RedCurve {keys [<at 0 0.3 RCIM_Cubic>]}
      }})
    R3 Orange:LinearColor=&ColorFromCurve(ColorCurve=DataInterface() {propertyValues {
       AlphaCurve {keys [<at 0 8 RCIM_Cubic>,<at 1 0 RCIM_Cubic>]}
       BlueCurve {keys [<at 0 0.0014 RCIM_Cubic>]}
       GreenCurve {keys [<at 0 0.0704 RCIM_Cubic>]}
       RedCurve {keys [<at 0 0.26 RCIM_Cubic>]}
      }})
    R3 Pink:LinearColor=&ColorFromCurve(ColorCurve=DataInterface() {propertyValues {
       AlphaCurve {keys [<at 0 8 RCIM_Cubic>,<at 1 0 RCIM_Cubic>]}
       BlueCurve {keys [<at 0 0.2604 RCIM_Cubic>]}
       GreenCurve {keys [<at 0 0.0453 RCIM_Cubic>]}
       RedCurve {keys [<at 0 0.36 RCIM_Cubic>]}
      }})
    R3 Purple:LinearColor=&ColorFromCurve(ColorCurve=DataInterface() {propertyValues {
       AlphaCurve {keys [<at 0 15 RCIM_Cubic>]}
       BlueCurve {keys [<at 0 0.35 RCIM_Cubic>]}
       GreenCurve {keys [<at 0 0.0455 RCIM_Cubic>]}
       RedCurve {keys [<at 0 0.0963 RCIM_Cubic>]}
      }})
    R3 Saffron:LinearColor=&ColorFromCurve(ColorCurve=DataInterface() {propertyValues {
       AlphaCurve {keys [<at 0.001 10 RCIM_Cubic>,<at 0.9493 0 RCIM_Cubic>]}
       BlueCurve {keys [<at 0 0 RCIM_Cubic>]}
       GreenCurve {keys [<at 0 0.25 RCIM_Cubic>]}
       RedCurve {keys [<at 0 0.2269 RCIM_Cubic>]}
      }})
    R3 SkyBlue:LinearColor=&ColorFromCurve(ColorCurve=DataInterface() {propertyValues {
       AlphaCurve {keys [<at 0 8 RCIM_Cubic>,<at 1 0 RCIM_Cubic>]}
       BlueCurve {keys [<at 0 0.3 RCIM_Cubic>]}
       GreenCurve {keys [<at 0 0.237 RCIM_Cubic>]}
       RedCurve {keys [<at 0 0 RCIM_Cubic>]}
      }})
    R3 TitaniumWhite:LinearColor=&ColorFromCurve(ColorCurve=DataInterface() {propertyValues {
       AlphaCurve {keys [<at 0 8 RCIM_Cubic>,<at 1 0 RCIM_Cubic>]}
       BlueCurve {keys [<at 0 0.35 RCIM_Cubic>]}
       GreenCurve {keys [<at 0 0.35 RCIM_Cubic>]}
       RedCurve {keys [<at 0 0.35 RCIM_Cubic>]}
      }})
    R4 Black:LinearColor=<rgba 0.0191 0.0191 0.0191 0>
    R4 BurntSienna:LinearColor=<rgba 0.0642 0.0198 0.0079 0>
    R4 Cobalt:LinearColor=&ColorFromCurve(ColorCurve=DataInterface() {propertyValues {
       AlphaCurve {keys [<at 0 5 RCIM_Cubic>,<at 1 0 RCIM_Cubic>]}
       BlueCurve {keys [<at 0 0 RCIM_Cubic>]}
       GreenCurve {keys [<at 0 0 RCIM_Cubic>]}
       RedCurve {keys [<at 0 0 RCIM_Cubic>]}
      }})
    R4 Crimson:LinearColor=&ColorFromCurve(ColorCurve=DataInterface() {propertyValues {
       AlphaCurve {keys [<at 0 6 RCIM_Cubic>,<at 1 0 RCIM_Cubic>]}
       BlueCurve {keys [<at 0 0.0001 RCIM_Cubic>]}
       GreenCurve {keys [<at 0 0 RCIM_Cubic>]}
       RedCurve {keys [<at 0 0.01 RCIM_Cubic>]}
      }})
    R4 DefaultColor:LinearColor=<rgba 1 1 1 0>
    R4 ForestGreen:LinearColor=<rgba 0 0.5 0.0253 0>
    R4 Grey:LinearColor=<rgba 0.2031 0.2031 0.2031 0>
    R4 Lime:LinearColor=<rgba 0.2943 0.5885 0 0>
    R4 Orange:LinearColor=&ColorFromCurve(ColorCurve=DataInterface() {propertyValues {
       AlphaCurve {keys [<at 0 8 RCIM_Cubic>,<at 1 0 RCIM_Cubic>]}
       BlueCurve {keys [<at 0 0 RCIM_Cubic>]}
       GreenCurve {keys [<at 0 0 RCIM_Cubic>]}
       RedCurve {keys [<at 0 0 RCIM_Cubic>]}
      }})
    R4 Pink:LinearColor=&ColorFromCurve(ColorCurve=DataInterface() {propertyValues {
       AlphaCurve {keys [<at 0 8 RCIM_Cubic>,<at 1 0 RCIM_Cubic>]}
       BlueCurve {keys [<at 0 0 RCIM_Cubic>]}
       GreenCurve {keys [<at 0 0 RCIM_Cubic>]}
       RedCurve {keys [<at 0 0 RCIM_Cubic>]}
      }})
    R4 Purple:LinearColor=&ColorFromCurve(ColorCurve=DataInterface() {propertyValues {
       AlphaCurve {keys [<at 0 15 RCIM_Cubic>,<at 0.9966 0>]}
       BlueCurve {keys [<at 0 0.6671 RCIM_Cubic>]}
       GreenCurve {keys [<at 0 0.0667 RCIM_Cubic>]}
       RedCurve {keys [<at 0 0.1668 RCIM_Cubic>]}
      }})
    R4 Saffron:LinearColor=<rgba 0.9 0.69 0 0>
    R4 SkyBlue:LinearColor=<rgba 0 0.357 0.5365 0>
    R4 TitaniumWhite:LinearColor=&ColorFromCurve(ColorCurve=DataInterface() {propertyValues {
       AlphaCurve {keys [<at 0 15 RCIM_Cubic>,<at 1 0 RCIM_Cubic>]}
       BlueCurve {keys [<at 0 0 RCIM_Cubic>]}
       GreenCurve {keys [<at 0 0 RCIM_Cubic>]}
       RedCurve {keys [<at 0 0 RCIM_Cubic>]}
      }})
    Saffron=&ColorFromCurve(ColorCurve=DataInterface() {propertyValues {
       AlphaCurve {keys [<at 0.0242 0>,<at 0.0868 1 RCIM_Cubic>]}
       BlueCurve {keys [<at 0 0 RCIM_Cubic>,<at 0.9064 0>]}
       GreenCurve {keys [<at 0 0.4 RCIM_Cubic>,<at 0.9064 0.2>]}
       RedCurve {keys [<at 0 0.3813 RCIM_Cubic>,<at 0.9064 0.1906>]}
      }})
    SkyBlue=&ColorFromCurve(ColorCurve=DataInterface() {propertyValues {
       AlphaCurve {keys [<at 0.0216 0>,<at 0.0894 1 RCIM_Cubic>]}
       BlueCurve {keys [<at 0 0.6 RCIM_Cubic>,<at 0.7729 0.3>]}
       GreenCurve {keys [<at 0 0.2507 RCIM_Cubic>,<at 0.7729 0.11>]}
       RedCurve {keys [<at 0 0 RCIM_Cubic>,<at 0.7729 0>]}
      }})
    TitaniumWhite=&ColorFromCurve(ColorCurve=DataInterface() {propertyValues {
       AlphaCurve {keys [<at 0.0286 0>,<at 0.0797 1 RCIM_Cubic>]}
       BlueCurve {keys [<at 0 0.7529 RCIM_Cubic>,<at 0.489 0.4>]}
       GreenCurve {keys [<at 0 0.7529 RCIM_Cubic>,<at 0.489 0.4>]}
       RedCurve {keys [<at 0 0.7529 RCIM_Cubic>,<at 0.489 0.4>]}
      }})
    Use R2 Painted Overrides=true
    Use R3 Painted Overrides=true
    Use R4 Painted Overrides=true
   )
   &InheritSourceMovement(
    Applied Position Scale=&VectorFromFloat(Value=&FloatFromCurve(FloatCurve=DataInterface() {propertyValues {Curve {keys [
          {interpMode=RCIM_Cubic, tangentMode=RCTM_Break, time=0, value=1}
          {arriveTangent=-2.9025, interpMode=RCIM_Cubic, tangentMode=RCTM_Break, time=1, value=0}
         ]}}}))
    ApplyForce=false
    ApplyPosition=true
    ApplyPositionFalloff=false
   )
   &SolveForcesAndVelocity(Clamp Velocity=false, Limit Acceleration=false)
   &VC_UpdateRibbon_v1(Enable Skeletal Position=true)
   &ScaleRibbonWidth(Ribbon Width Scale=&FloatFromCurve(
     FloatCurve=DataInterface() {propertyValues {Curve {keys [
         {arriveTangent=0.0349, interpMode=RCIM_Cubic, leaveTangent=0.0349, tangentMode=RCTM_User, time=0, value=0.2}
         {arriveTangent=-0.116, tangentMode=RCTM_User, time=1, value=1}
        ]}}}
     Scale Curve=1.2
    ))
   /SetVariables(
    Particles.Ribbon 2 Width:NiagaraFloat=&Multiply_Float(
     A=$Emitter.Ribbon 2 Width
     B=&FloatFromCurve(FloatCurve=DataInterface() {propertyValues {Curve {keys [<at 0 0.2 RCIM_Cubic>,<at 1 1>]}}})
    )
    Particles.Ribbon 3 Width:NiagaraFloat=&Multiply_Float(
     A=$Emitter.Ribbon 3 Width
     B=&FloatFromCurve(FloatCurve=DataInterface() {propertyValues {Curve {keys [<at 0 1>,<at 0.5 4.5>]}}}, Scale Curve=0.9)
    )
   ) {isSetParameters=true}
   /SetVariables(Particles.PositionLightning:NiagaraPosition=&AddVectorToPosition(
     Position=$Particles.Position
     Vector=&TransformVector(
      Vector=&MakeVector(
       X=0
       Y=0
       Z=&Sine(
        Normalized Angle=&Add_Float(
         A=&Multiply_Float(A=$Particles.NormalizedAge, B=-1)
         B=&Multiply_Float(A=$Emitter.LoopedAge, B=1)
        )
        Period=1
        Scale=&FloatFromCurve(
         FloatCurve=DataInterface() {propertyValues {Curve {
            defaultValue=0
            keys [
             {interpMode=RCIM_Cubic, tangentMode=RCTM_User, time=0, value=0}
             {interpMode=RCIM_Cubic, tangentMode=RCTM_User, time=0.2, value=10}
             <at 0.4 -10 RCIM_Cubic>
             <at 0.6 10 RCIM_Cubic>
             <at 0.7 -10 RCIM_Cubic>
             {arriveTangent=0.5014, interpMode=RCIM_Cubic, leaveTangent=0.5014, tangentMode=RCTM_User, time=0.8, value=10}
             {tangentMode=RCTM_User, time=1, value=0}
            ]
           }}}
         Scale Curve=0.4
        )
        Bias=1
       )
      )
      Destination Space=::Simulation
      Apply Scale=true
     )
    )) {isSetParameters=true}
  ]
  rendererClasses [&NiagaraRibbonRendererProperties]
  renderers [
   {rendererClass=&NiagaraRibbonRendererProperties}
   {rendererClass=&NiagaraRibbonRendererProperties, rendererIndex=1}
   {rendererClass=&NiagaraRibbonRendererProperties, rendererIndex=2}
   {rendererClass=&NiagaraRibbonRendererProperties, rendererIndex=3}
  ]
 }}
summary {
 systemName=NS_Boost_Techno_Optimized
 emitters [{emitterName=TrailPlasmaEnergy, rendererClasses [&NiagaraRibbonRendererProperties]}]
 userVariables [
  User.Alpha_line {type=NiagaraFloat, defaultValue=1, description=""}
  User.bIsFrontendPreview {type=NiagaraBool}
  User.bIsPainted {type=NiagaraBool}
  User.black {type=NiagaraFloat, defaultValue=1.5, description=""}
  User.ColorVariantID {type=NiagaraInt32}
  User.Degradadoline {type=NiagaraFloat, defaultValue=1, description=""}
  User.DTCarOffset {type=Vector3f, description="", defaultValue {struct=Vector3f, value=<0 0 0>}}
  User.Intensity {type=NiagaraFloat, defaultValue=1, description=""}
  User.PaintedColor {type=LinearColor, description="", defaultValue {struct=LinearColor, value=<rgba 1 1 1 1>}}
  User.PaintedColor2 {type=LinearColor, description="", defaultValue {struct=LinearColor, value=<rgba 1 1 1 1>}}
  User.PaintedScalar {type=NiagaraFloat, defaultValue=1, description=""}
  User.PortCount {type=NiagaraInt32, defaultValue=1}
  User.Power_lightning {type=NiagaraFloat, defaultValue=1, description=""}
  User.SingleBoostPos {type=Vector3f, description="", defaultValue {struct=Vector3f, value=<0 0 0>}}
 ]
}
